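# Animation Blueprint for the StunningMountain backpack, animating its six floating pieces.
# Spring Anim Dynamics, weakened while skydiving, combine with time-based sine bobbing into a cached pose.
# Additive bone offsets then adjust that pose for spine bend and jumping or falling; computed clavicle alphas go unused.
# Weapon, customization and skydiving events toggle material parameter "Hide Element 01", set when a tagged pickaxe is equipped outside skydiving.
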AnimBlueprint /BRCosmetics/Accessories/FORT_Backpacks/Backpack_M_MED_StunningMountain/Meshes/M_MED_StunningMountain_Pack_AnimBP
target_skeleton=/Game/Accessories/FORT_Backpacks/Backpack_M_MED_StunningMountain/Meshes/M_MED_StunningMountain_Pack_Skeleton
graphs [
 AnimGraph {nodes [
   n1 = "Transform (Modify) Bone - Bone: pack"!(
    ComponentPose=Local To Component(LocalPose=Use cached pose 'Post Idle Setup'())
    bAlphaBoolEnabled:bool
    Translation="0,-1,-2"
    Rotation="0,0,0"
    Scale
   ) -> (Pose) {
    alphaInputType=Bool
    translationMode=BMM_Additive
    translationSpace=BCS_BoneSpace
    alphaBoolBlend {blendInTime=0.3, blendOption=ExpInOut, blendOutTime=0.3}
    boneToModify {boneName=pack}
    translation {y=-1, z=-2}
   }
   n2 = Get SpineCurveCompensationAlpha() -> (SpineCurveCompensationAlpha:double)
   n3 = "Transform (Modify) Bone - Bone: dyn_floaties_MSTR"(ComponentPose=@n1, Alpha=@n2, Translation="3,0,0", Rotation="35.000,0.000,0.000", Scale="1,1,1") {
    rotationMode=BMM_Additive
    rotationSpace=BCS_BoneSpace
    translationMode=BMM_Additive
    translationSpace=BCS_BoneSpace
    alphaBoolBlend {blendOption=Linear}
    boneToModify {boneName=dyn_floaties_MSTR}
    rotation=<rot 34.9999 0 0>
    translation {x=3}
   }
   n4 = "Transform (Modify) Bone - Bone: dyn_floaty_lower_left"(ComponentPose=@n3, Alpha=@n2, Translation="2.8,-0.5,0.2", Rotation="30,0,0", Scale="1,1,1") {
    rotationMode=BMM_Additive
    rotationSpace=BCS_BoneSpace
    translationMode=BMM_Additive
    translationSpace=BCS_BoneSpace
    alphaBoolBlend {blendOption=Linear}
    boneToModify {boneName=dyn_floaty_lower_left}
    rotation {pitch=30}
    translation=<2.8 -0.5 0.2>
   }
   n5 = "Transform (Modify) Bone - Bone: dyn_floaty_lower_right"(ComponentPose=@n4, Alpha=@n2, Translation="2.8,0.5,0.588", Rotation="30,0,0", Scale="1,1,1") {
    rotationMode=BMM_Additive
    rotationSpace=BCS_BoneSpace
    translationMode=BMM_Additive
    translationSpace=BCS_BoneSpace
    alphaBoolBlend {blendOption=Linear}
    boneToModify {boneName=dyn_floaty_lower_right}
    rotation {pitch=30}
    translation=<2.8 0.5 0.588>
   }
   n6 = "Transform (Modify) Bone - Bone: dyn_floaty_lower_center"(ComponentPose=@n5, Alpha=@n2, Translation="2,0,0", Rotation="20.000,0.000,0.000", Scale="1,1,1") {
    rotationMode=BMM_Additive
    rotationSpace=BCS_BoneSpace
    translationMode=BMM_Additive
    translationSpace=BCS_BoneSpace
    alphaBoolBlend {blendOption=Linear}
    boneToModify {boneName=dyn_floaty_lower_center}
    rotation=<rot 20 0 0>
    translation {x=2}
   }
   n7 = "Transform (Modify) Bone - Bone: dyn_floaty_upper_right"(ComponentPose=@n6, Alpha=1, Translation="3,2,0", Rotation="-0.000,35.000,-0.000", Scale="1,1,1") {
    rotationMode=BMM_Additive
    rotationSpace=BCS_BoneSpace
    translationMode=BMM_Additive
    translationSpace=BCS_BoneSpace
    alphaBoolBlend {blendOption=Linear}
    boneToModify {boneName=dyn_floaty_upper_right}
    rotation=<rot -0.0001 35 -0.0001>
    translation=<3 2>
   }
   n8 = "Transform (Modify) Bone - Bone: dyn_floaty_upper_left"(ComponentPose=@n7, Alpha=1, Translation="3,-2,0", Rotation="0,-35,0", Scale="1,1,1") {
    rotationMode=BMM_Additive
    rotationSpace=BCS_BoneSpace
    translationMode=BMM_Additive
    translationSpace=BCS_BoneSpace
    alphaBoolBlend {blendOption=Linear}
    boneToModify {boneName=dyn_floaty_upper_left}
    rotation {yaw=-35}
    translation=<3 -2>
   }
   n9 = Get Is Jumping or Falling() -> (Is Jumping or Falling:bool)
   n10 = "Transform (Modify) Bone - Bone: dyn_floaty_upper_right"!(ComponentPose=@n8, bAlphaBoolEnabled:bool=@n9, Translation="0.000,-1.555,1.851", Rotation="0,0,0", Scale="1,1,1") -> (Pose) {
    alphaInputType=Bool
    translationMode=BMM_Additive
    translationSpace=BCS_BoneSpace
    alphaBoolBlend {blendInTime=0.6, blendOutTime=0.6}
    alphaScaleBiasClamp {bInterpResult=true, interpSpeedDecreasing=5, interpSpeedIncreasing=5}
    boneToModify {boneName=dyn_floaty_upper_right}
    translation {y=-1.555, z=1.851}
   }
   n11 = "Transform (Modify) Bone - Bone: dyn_floaty_lower_center"!(ComponentPose=@n10, bAlphaBoolEnabled:bool=@n9, Translation="0.000,0.000,-1.124", Rotation="0,0,0", Scale="1,1,1") -> (Pose) {
    alphaInputType=Bool
    translationMode=BMM_Additive
    translationSpace=BCS_BoneSpace
    alphaBoolBlend {blendInTime=0.6, blendOutTime=0.6}
    alphaScaleBiasClamp {bInterpResult=true, interpSpeedDecreasing=5, interpSpeedIncreasing=5}
    boneToModify {boneName=dyn_floaty_lower_center}
    translation {z=-1.124}
   }
   n12 = "Transform (Modify) Bone - Bone: dyn_floaty_upper_center"!(ComponentPose=@n11, bAlphaBoolEnabled:bool=@n9, Translation="0,0,0", Rotation="0,0,0", Scale="1,1,1") -> (Pose) {
    alphaInputType=Bool
    translationMode=BMM_Additive
    translationSpace=BCS_BoneSpace
    alphaBoolBlend {blendInTime=0.6, blendOutTime=0.6}
    alphaScaleBiasClamp {bInterpResult=true, interpSpeedDecreasing=5, interpSpeedIncreasing=5}
    boneToModify {boneName=dyn_floaty_upper_center}
   }
   n13 = "Transform (Modify) Bone - Bone: dyn_floaty_lower_left"!(ComponentPose=@n12, bAlphaBoolEnabled:bool=@n9, Translation="0.000,0.922,-1.563", Rotation="0,0,0", Scale="1,1,1") -> (Pose) {
    alphaInputType=Bool
    translationMode=BMM_Additive
    translationSpace=BCS_BoneSpace
    alphaBoolBlend {blendInTime=0.6, blendOutTime=0.6}
    alphaScaleBiasClamp {bInterpResult=true, interpSpeedDecreasing=5, interpSpeedIncreasing=5}
    boneToModify {boneName=dyn_floaty_lower_left}
    translation {y=0.922, z=-1.563}
   }
   n14 = "Transform (Modify) Bone - Bone: dyn_floaty_upper_left"!(ComponentPose=@n13, bAlphaBoolEnabled:bool=@n9, Translation="0.000,0.001,2.419", Rotation="0,0,0", Scale="1,1,1") -> (Pose) {
    alphaInputType=Bool
    translationMode=BMM_Additive
    translationSpace=BCS_BoneSpace
    alphaBoolBlend {blendInTime=0.6, blendOutTime=0.6}
    alphaScaleBiasClamp {bInterpResult=true, interpSpeedDecreasing=5, interpSpeedIncreasing=5}
    boneToModify {boneName=dyn_floaty_upper_left}
    translation {y=0.001, z=2.419}
   }
   n15 = "Transform (Modify) Bone - Bone: dyn_floaty_lower_right"!(ComponentPose=@n14, bAlphaBoolEnabled:bool=@n9, Translation="-0.000,-1.380,-1.002", Rotation="0,0,0", Scale="1,1,1") -> (Pose) {
    alphaInputType=Bool
    translationMode=BMM_Additive
    translationSpace=BCS_BoneSpace
    alphaBoolBlend {blendInTime=0.6, blendOutTime=0.6}
    alphaScaleBiasClamp {bInterpResult=true, interpSpeedDecreasing=5, interpSpeedIncreasing=5}
    boneToModify {boneName=dyn_floaty_lower_right}
    translation {y=-1.38, z=-1.002}
   }
   Output Pose(Result=Component To Local(ComponentPose=@n15)) {layerGroup=None}
   n16 = "Transform (Modify) Bone - Bone: pack"!(ComponentPose=Local To Component(LocalPose=Input Pose()), Alpha:float, Translation="0,-2,0", Rotation="0,0,0", Scale) -> (Pose) {
    translationMode=BMM_Additive
    translationSpace=BCS_BoneSpace
    alphaBoolBlend {blendOption=Linear}
    boneToModify {boneName=pack}
    translation {y=-2}
   }
   n17 = Get Anim Dynamics Alpha() -> (Anim Dynamics Alpha:double)
   n18 = Anim Dynamics - Bone: dyn_floaty_upper_center!(
    ComponentPose=@n16
    Alpha=@n17
    GravityOverride:Vector="0,0,20"
    ComponentLinearAccScale:Vector="0.7,0.7,1.5"
    ComponentLinearVelScale:Vector="0.7,0.7,1.5"
    ExternalForce="0,0,0"
   ) -> (Pose) {
    alpha=0
    angularBiasOverride=1
    angularDampingOverride=0.75
    angularSpringConstant=20
    bGravityOverrideInSimSpace=true
    bLinearSpring=true
    bOverrideLinearDamping=true
    bUseGravityOverride=true
    bUsePlanarLimit=false
    gravityScale=0
    linearDampingOverride=0.9
    linearSpringConstant=30
    simulationSpace=BoneRelative
    alphaBoolBlend {blendOption=Linear}
    alphaScaleBiasClamp {bInterpResult=true}
    boundBone {boneName=dyn_floaty_upper_center}
    componentAppliedLinearAccClamp=<63 53 47>
    componentLinearAccScale=<0.7 0.7 1.5>
    componentLinearVelScale=<0.7 0.7 1.5>
    gravityOverride {z=20}
    physicsBodyDefinitions [{
      sphereCollisionRadius=0
      boundBone {boneName=dyn_floaty_upper_center}
      boxExtents=<22 22 22>
      constraintSetup {
       linearXLimitType=Free
       linearZLimitType=Free
       angularLimitsMax=<0 0 0>
       angularLimitsMin=<0 0 0>
       angularTarget=<0 0 0>
       linearAxesMax=<2.5 0.5 0>
       linearAxesMin=<-2.5 -10 0>
      }
      localJointOffset=<0 0 0>
     }]
    relativeSpaceBone {boneName=root}
   }
   n19 = Anim Dynamics - Bone: dyn_floaty_lower_left!(
    ComponentPose=@n18
    Alpha=@n17
    GravityOverride:Vector="0,0,20"
    ComponentLinearAccScale:Vector="0.7,0.7,1.5"
    ComponentLinearVelScale:Vector="0.7,0.7,1.5"
    ExternalForce="0,0,0"
   ) -> (Pose) {
    alpha=0
    angularBiasOverride=1
    angularDampingOverride=0.75
    angularSpringConstant=20
    bGravityOverrideInSimSpace=true
    bLinearSpring=true
    bOverrideLinearDamping=true
    bUseGravityOverride=true
    bUsePlanarLimit=false
    gravityScale=0
    linearDampingOverride=0.9
    linearSpringConstant=18
    simulationSpace=BoneRelative
    alphaBoolBlend {blendOption=Linear}
    alphaScaleBiasClamp {bInterpResult=true}
    boundBone {boneName=dyn_floaty_lower_left}
    componentAppliedLinearAccClamp=<63 53 47>
    componentLinearAccScale=<0.7 0.7 1.5>
    componentLinearVelScale=<0.7 0.7 1.5>
    gravityOverride {z=20}
    physicsBodyDefinitions [{
      sphereCollisionRadius=0
      boundBone {boneName=dyn_floaty_lower_left}
      boxExtents=<22 22 22>
      constraintSetup {
       linearXLimitType=Free
       linearZLimitType=Free
       angularLimitsMax=<0 0 0>
       angularLimitsMin=<0 0 0>
       angularTarget=<0 0 0>
       linearAxesMax=<2.5 0.5 0>
       linearAxesMin=<-2.5 -10 0>
      }
      localJointOffset=<0 0 0>
     }]
    relativeSpaceBone {boneName=root}
   }
   n20 = Anim Dynamics - Bone: dyn_floaty_lower_right!(
    ComponentPose=@n19
    Alpha=@n17
    GravityOverride:Vector="0,0,20"
    ComponentLinearAccScale:Vector="0.7,0.7,1.5"
    ComponentLinearVelScale:Vector="0.7,0.7,1.5"
    ExternalForce="0,0,0"
   ) -> (Pose) {
    alpha=0
    angularBiasOverride=1
    angularDampingOverride=0.75
    angularSpringConstant=20
    bGravityOverrideInSimSpace=true
    bLinearSpring=true
    bOverrideLinearDamping=true
    bUseGravityOverride=true
    bUsePlanarLimit=false
    gravityScale=0
    linearDampingOverride=0.9
    linearSpringConstant=24
    simulationSpace=BoneRelative
    alphaBoolBlend {blendOption=Linear}
    alphaScaleBiasClamp {bInterpResult=true}
    boundBone {boneName=dyn_floaty_lower_right}
    componentAppliedLinearAccClamp=<63 53 47>
    componentLinearAccScale=<0.7 0.7 1.5>
    componentLinearVelScale=<0.7 0.7 1.5>
    gravityOverride {z=20}
    physicsBodyDefinitions [{
      sphereCollisionRadius=0
      boundBone {boneName=dyn_floaty_lower_right}
      boxExtents=<22 22 22>
      constraintSetup {
       linearXLimitType=Free
       linearZLimitType=Free
       angularLimitsMax=<0 0 0>
       angularLimitsMin=<0 0 0>
       angularTarget=<0 0 0>
       linearAxesMax=<2.5 0.5 0>
       linearAxesMin=<-2.5 -10 0>
      }
      localJointOffset=<0 0 0>
     }]
    relativeSpaceBone {boneName=root}
   }
   n21 = Anim Dynamics - Bone: dyn_floaty_lower_center!(
    ComponentPose=@n20
    Alpha=@n17
    GravityOverride:Vector="0,0,20"
    ComponentLinearAccScale:Vector="0.7,0.7,1.5"
    ComponentLinearVelScale:Vector="0.7,0.7,1.5"
    ExternalForce="0,0,0"
   ) -> (Pose) {
    alpha=0
    angularBiasOverride=1
    angularDampingOverride=0.75
    angularSpringConstant=20
    bGravityOverrideInSimSpace=true
    bLinearSpring=true
    bOverrideLinearDamping=true
    bUseGravityOverride=true
    bUsePlanarLimit=false
    gravityScale=0
    linearDampingOverride=0.99
    linearSpringConstant=40
    simulationSpace=BoneRelative
    alphaBoolBlend {blendOption=Linear}
    alphaScaleBiasClamp {bInterpResult=true}
    boundBone {boneName=dyn_floaty_lower_center}
    componentAppliedLinearAccClamp=<63 53 47>
    componentLinearAccScale=<0.7 0.7 1.5>
    componentLinearVelScale=<0.7 0.7 1.5>
    gravityOverride {z=20}
    physicsBodyDefinitions [{
      sphereCollisionRadius=0
      boundBone {boneName=dyn_floaty_lower_center}
      boxExtents=<22 22 22>
      constraintSetup {
       linearXLimitType=Free
       linearZLimitType=Free
       angularLimitsMax=<0 0 0>
       angularLimitsMin=<0 0 0>
       angularTarget=<0 0 0>
       linearAxesMax=<2.5 0.5 0>
       linearAxesMin=<-2.5 -10 0>
      }
      localJointOffset=<0 0 0>
     }]
    relativeSpaceBone {boneName=root}
   }
   n22 = Anim Dynamics - Bone: dyn_floaty_upper_right!(
    ComponentPose=@n21
    Alpha=@n17
    GravityOverride:Vector="0,0,20"
    ComponentLinearAccScale:Vector="0.7,0.7,1.5"
    ComponentLinearVelScale:Vector="0.7,0.7,1.5"
    ExternalForce="0,0,0"
   ) -> (Pose) {
    alpha=0
    angularBiasOverride=1
    angularDampingOverride=0.75
    angularSpringConstant=20
    bGravityOverrideInSimSpace=true
    bLinearSpring=true
    bOverrideLinearDamping=true
    bUseGravityOverride=true
    bUsePlanarLimit=false
    gravityScale=0
    linearDampingOverride=0.9
    linearSpringConstant=50
    simulationSpace=BoneRelative
    alphaBoolBlend {blendOption=Linear}
    alphaScaleBiasClamp {bInterpResult=true}
    boundBone {boneName=dyn_floaty_upper_right}
    componentAppliedLinearAccClamp=<63 53 47>
    componentLinearAccScale=<0.7 0.7 1.5>
    componentLinearVelScale=<0.7 0.7 1.5>
    gravityOverride {z=20}
    physicsBodyDefinitions [{
      sphereCollisionRadius=0
      boundBone {boneName=dyn_floaty_upper_right}
      boxExtents=<22 22 22>
      constraintSetup {
       linearXLimitType=Free
       linearZLimitType=Free
       angularLimitsMax=<0 0 0>
       angularLimitsMin=<0 0 0>
       angularTarget=<0 0 0>
       linearAxesMax=<2.5 0.5 0>
       linearAxesMin=<-2.5 -10 0>
      }
      localJointOffset=<0 0 0>
     }]
    relativeSpaceBone {boneName=root}
   }
   n23 = Anim Dynamics - Bone: dyn_floaty_upper_left!(
    ComponentPose=@n22
    Alpha=@n17
    GravityOverride:Vector="0,0,20"
    ComponentLinearAccScale:Vector="0.7,0.7,1.5"
    ComponentLinearVelScale:Vector="0.7,0.7,1.5"
    ExternalForce="0,0,0"
   ) -> (Pose) {
    alpha=0
    angularBiasOverride=1
    angularDampingOverride=0.75
    angularSpringConstant=20
    bGravityOverrideInSimSpace=true
    bLinearSpring=true
    bOverrideLinearDamping=true
    bUseGravityOverride=true
    bUsePlanarLimit=false
    gravityScale=0
    linearDampingOverride=0.9
    linearSpringConstant=30
    simulationSpace=BoneRelative
    alphaBoolBlend {blendOption=Linear}
    alphaScaleBiasClamp {bInterpResult=true}
    boundBone {boneName=dyn_floaty_upper_left}
    componentAppliedLinearAccClamp=<63 53 47>
    componentLinearAccScale=<0.7 0.7 1.5>
    componentLinearVelScale=<0.7 0.7 1.5>
    gravityOverride {z=20}
    physicsBodyDefinitions [{
      sphereCollisionRadius=0
      boundBone {boneName=dyn_floaty_upper_left}
      boxExtents=<22 22 22>
      constraintSetup {
       linearXLimitType=Free
       linearZLimitType=Free
       angularLimitsMax=<0 0 0>
       angularLimitsMin=<0 0 0>
       angularTarget=<0 0 0>
       linearAxesMax=<2.5 0.5 0>
       linearAxesMin=<-2.5 -10 0>
      }
      localJointOffset=<0 0 0>
     }]
    relativeSpaceBone {boneName=root}
   }
   n24 = "Transform (Modify) Bone - Bone: dyn_floaty_upper_right"!(
    ComponentPose=@n23
    Alpha:float=1
    Translation=Get UpRight Pieces floating() -> (UpRight Pieces floating:Vector)
    Rotation="0,0,0"
    Scale="1,1,1"
   ) -> (Pose) {
    translationMode=BMM_Additive
    translationSpace=BCS_BoneSpace
    alphaBoolBlend {blendOption=Linear}
    boneToModify {boneName=dyn_floaty_upper_right}
   }
   n25 = "Transform (Modify) Bone - Bone: dyn_floaty_lower_center"!(
    ComponentPose=@n24
    Alpha:float=1
    Translation=Get LowCenter Pieces floating() -> (LowCenter Pieces floating:Vector)
    Rotation="0,0,0"
    Scale="1,1,1"
   ) -> (Pose) {
    translationMode=BMM_Additive
    translationSpace=BCS_BoneSpace
    alphaBoolBlend {blendOption=Linear}
    boneToModify {boneName=dyn_floaty_lower_center}
   }
   n26 = "Transform (Modify) Bone - Bone: dyn_floaty_upper_center"!(
    ComponentPose=@n25
    Alpha:float=1
    Translation=Get UpCenter Pieces floating() -> (UpCenter Pieces floating:Vector)
    Rotation="0,0,0"
    Scale="1,1,1"
   ) -> (Pose) {
    translationMode=BMM_Additive
    translationSpace=BCS_BoneSpace
    alphaBoolBlend {blendOption=Linear}
    boneToModify {boneName=dyn_floaty_upper_center}
   }
   n27 = "Transform (Modify) Bone - Bone: dyn_floaty_lower_left"!(
    ComponentPose=@n26
    Alpha:float=1
    Translation=Get LowLeft Pieces floating() -> (LowLeft Pieces floating:Vector)
    Rotation="0,0,0"
    Scale="1,1,1"
   ) -> (Pose) {
    translationMode=BMM_Additive
    translationSpace=BCS_BoneSpace
    alphaBoolBlend {blendOption=Linear}
    boneToModify {boneName=dyn_floaty_lower_left}
   }
   n28 = "Transform (Modify) Bone - Bone: dyn_floaty_upper_left"!(
    ComponentPose=@n27
    Alpha:float=1
    Translation=Get UpLeft Pieces floating() -> (UpLeft Pieces floating:Vector)
    Rotation="0,0,0"
    Scale="1,1,1"
   ) -> (Pose) {
    translationMode=BMM_Additive
    translationSpace=BCS_BoneSpace
    alphaBoolBlend {blendOption=Linear}
    boneToModify {boneName=dyn_floaty_upper_left}
   }
   n29 = "Transform (Modify) Bone - Bone: dyn_floaty_lower_right"!(
    ComponentPose=@n28
    Alpha:float=1
    Translation=Get LowRight Pieces floating() -> (LowRight Pieces floating:Vector)
    Rotation="0,0,0"
    Scale="1,1,1"
   ) -> (Pose) {
    translationMode=BMM_Additive
    translationSpace=BCS_BoneSpace
    alphaBoolBlend {blendOption=Linear}
    boneToModify {boneName=dyn_floaty_lower_right}
   }
   Save cached pose 'Post Idle Setup'(Pose=Component To Local(ComponentPose=@n29)) {cacheName=Post Idle Setup}
  ]}
 EventGraph {nodes [
   Event BlueprintUpdateAnimation()
   > n1 = Sequence!(execute) -> (then_0, then_1, then_2:Exec)
   Set Is Jumping or Falling(execute=@n1.then_1, Is Jumping or Falling:bool=OR Boolean(A=Get bIsJumping(), B=Get bIsFalling())) -> (Output_Get:bool)
   > Set Anim Dynamics Alpha(Anim Dynamics Alpha:double=SelectFloat(A=0.1, B=0.5, bPickA=Get bIsSkyDiving())) -> (Output_Get:double)
   n2 = Sequence(execute=@n1.then_2)
   n3 = GetOwnerSkeletalMeshComponent()
   > Set SpineCurveCompensationAlpha(SpineCurveCompensationAlpha:double=GetAlphaFromReferenceTransformDelta(SkeletalMeshComponent=@n3, BoneName=spine_05, BaseName=spine_01, InputRange=(X=0,Y=-40), TwistAxis="1,0,0")) -> (Output_Get:double)
   Set ClavicleLeftCompensationAlpha(
    execute=@n2.then_1
    ClavicleLeftCompensationAlpha:double=GetAlphaFromReferenceTransformDelta(SkeletalMeshComponent=@n3, BoneName=clavicle_l, BaseName=spine_05, InputRange=(X=-5,Y=-35), TwistAxis="1,0,0")
   ) -> (Output_Get:double)
   Set ClavicleRightCompensationAlpha(
    execute=@n2.then_2
    ClavicleRightCompensationAlpha:double=GetAlphaFromReferenceTransformDelta(SkeletalMeshComponent=@n3, BoneName=clavicle_r, BaseName=spine_05, InputRange=(X=-5,Y=-35), TwistAxis="1,0,0")
   ) -> (Output_Get:double)
   n4 = float * float(A=GetGameTimeInSeconds(), B=1.1) {is=title K2Node_PromotableOperator}
   n5 = "Sin (Radians)"(A=float * float(A=1.15, B=@n4) {is=title K2Node_CommutativeAssociativeBinaryOperator})
   n6 = "Sin (Radians)"(A=float * float(A=0.6, B=@n4))
   n7 = float * float(A="Sin (Radians)"(A=float * float(A=0.6999, B=@n4)), B=1.4)
   Set UpCenter Pieces floating(
    execute=@n1.then_0
    UpCenter Pieces floating:Vector=MakeVector(X=float * float(A=@n6, B=0.5), Y=float * float(A=@n5, B=0.15), Z=float + float(A=@n7))
   ) -> (Output_Get:Vector)
   n8 = float * float(A=GetGameTimeInSeconds(), B=1.103) {is=title K2Node_PromotableOperator}
   n9 = float * float(A="Sin (Radians)"(A=float * float(A=0.6999, B=@n8)), B=1.1) {is=title K2Node_CommutativeAssociativeBinaryOperator}
   n10 = float * float(A="Sin (Radians)"(A=float * float(A=0.6, B=@n8)), B=0.3)
   n11 = float * float(A="Sin (Radians)"(A=float * float(A=1, B=@n8)), B=0.2)
   > Set UpRight Pieces floating(UpRight Pieces floating:Vector=MakeVector(X=float + float(A=@n10, B=1), Y=float + float(A=@n11, B=-1.4), Z=float + float(A=@n9, B=5))) -> (Output_Get:Vector)
   n12 = float * float(A=GetGameTimeInSeconds(), B=1.1) {is=title K2Node_PromotableOperator}
   n13 = "Sin (Radians)"(A=float * float(A=2, B=@n12) {is=title K2Node_CommutativeAssociativeBinaryOperator})
   n14 = float * float(A="Sin (Radians)"(A=float * float(A=0.6999, B=@n12)), B=1.4)
   n15 = float * float(A="Sin (Radians)"(A=float * float(A=0.6, B=@n12)), B=0.4)
   > Set LowRight Pieces floating(LowRight Pieces floating:Vector=MakeVector(X=float + float(A=@n15), Y=float * float(A=@n13, B=0.1), Z=float + float(A=@n14, B=0.5))) -> (Output_Get:Vector)
   n16 = float * float(A=GetGameTimeInSeconds(), B=1.1) {is=title K2Node_PromotableOperator}
   n17 = "Sin (Radians)"(A=float * float(A=2, B=@n16) {is=title K2Node_CommutativeAssociativeBinaryOperator})
   n18 = "Sin (Radians)"(A=float * float(A=0.6, B=@n16))
   n19 = float * float(A="Sin (Radians)"(A=float * float(A=0.6999, B=@n16)), B=1.9)
   > Set LowCenter Pieces floating(LowCenter Pieces floating:Vector=MakeVector(X=float * float(A=@n18, B=0.4), Y=float * float(A=@n17, B=0.15), Z=float + float(A=@n19))) -> (Output_Get:Vector)
   n20 = float * float(A=GetGameTimeInSeconds(), B=1.008) {is=title K2Node_PromotableOperator}
   n21 = float * float(A="Sin (Radians)"(A=float * float(A=2, B=@n20)), B=0.15) {is=title K2Node_CommutativeAssociativeBinaryOperator}
   n22 = float * float(A="Sin (Radians)"(A=float * float(A=0.6999, B=@n20)), B=1.7)
   n23 = "Sin (Radians)"(A=float * float(A=0.6, B=@n20))
   > Set LowLeft Pieces floating(LowLeft Pieces floating:Vector=MakeVector(X=float * float(A=@n23, B=0.4), Y=float + float(A=@n21), Z=float + float(A=@n22, B=-0.4))) -> (Output_Get:Vector)
   n24 = float * float(A=GetGameTimeInSeconds(), B=1.1) {is=title K2Node_PromotableOperator}
   n25 = float * float(A="Sin (Radians)"(A=float * float(A=0.6, B=@n24)), B=0.4) {is=title K2Node_CommutativeAssociativeBinaryOperator}
   n26 = float * float(A="Sin (Radians)"(A=float * float(A=1, B=@n24)), B=0.6)
   n27 = float * float(A="Sin (Radians)"(A=float * float(A=0.6999, B=@n24)), B=1.5)
   > Set UpLeft Pieces floating(UpLeft Pieces floating:Vector=MakeVector(X=float + float(A=@n25, B=1.5), Y=float + float(A=@n26, B=1.3), Z=float + float(A=@n27, B=3))) -> (Output_Get:Vector)
   n28 = Event OnCharacterCustomizationLoaded()
   n29 = Get Dynamic Material Instance(execute=[@n28.then, @n33.then], SkeletalMeshComponent=GetPartSkeletalMeshComponent())
   n30 = PickaxeItemDefinitionHasAnyMetaTag(Tags=Make Array("[0]"=(TagName="Cosmetics.MetaTags.StunningMountain")), bRequiresEquipped=true)
   > n31 = Branch(Condition=@n30)
   > n32 = Branch(Condition=Get bIsSkyDiving())
   SetScalarParameterValue(execute=[@n34.then, @n35.else, @n32.then, @n31.else], self=@n29, ParameterName=Hide Element 01, Value=0)
   SetScalarParameterValue(execute=[@n35.then, @n32.else], self=@n29, ParameterName=Hide Element 01, Value=1)
   n33 = Event OnWeaponChanged()
   n34 = Event OnBeginSkydiving()
   Event OnEndSkydiving()
   > n35 = Branch(Condition=@n30)
  ]}
]
variables [
 Anim Dynamics Alpha {type=number}
 ClavicleLeftCompensationAlpha {type=number}
 ClavicleRightCompensationAlpha {type=number}
 DynamicMaterialInstance {type=MaterialInstanceDynamic}
 Is Jumping or Falling {type=boolean}
 LowCenter Pieces floating {type=Vector}
 LowLeft Pieces floating {type=Vector}
 LowRight Pieces floating {type=Vector}
 SkeletalMesh {type=SkeletalMeshComponent}
 SpineCurveCompensationAlpha {type=number}
 UpCenter Pieces floating {type=Vector}
 UpLeft Pieces floating {type=Vector}
 UpRight Pieces floating {type=Vector}
]
